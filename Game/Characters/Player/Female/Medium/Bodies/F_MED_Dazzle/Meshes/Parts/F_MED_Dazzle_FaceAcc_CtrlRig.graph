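# Control Rig for the Dazzle female character's face accessory that repositions braid joints while the `ads` flag (likely aim down sights) is set.
# A spherical pose reader on head relative to spine_05 writes the `ads spr` weight every evaluation.
# When `ads` is true, clamped head Euler Z offsets spine and right-arm lookat bones.
# Rotation and aim constraints plus rotation offsets then pose selected master braid joints, weighted by `ads spr`.
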ControlRigBlueprint /Game/Characters/Player/Female/Medium/Bodies/F_MED_Dazzle/Meshes/Parts/F_MED_Dazzle_FaceAcc_CtrlRig
graphs [
 RigVMFunctionLibrary {links [], nodes []}
 RigVMModel {nodes [
   RigUnit_BeginExecution = &RigUnit_BeginExecution()
   SequenceAggregate = &RigVMFunction_Sequence(ExecuteContext=@RigUnit_BeginExecution.ExecutePin)
   n1 = &RigUnit_SphericalPoseReader(
    ActiveRegionScaleFactors=(PositiveWidth=1,NegativeWidth=1,PositiveHeight=1,NegativeHeight=1)
    ActiveRegionSize=0.3
    Debug=(bDrawDebug=false,DebugScale=25,DebugSegments=20,DebugThickness=0.25)
    DriverAxis=(X=1,Y=0,Z=0)
    DriverItem=(Type=Bone,Name="head")
    FalloffRegionScaleFactors=(PositiveWidth=1,NegativeWidth=1,PositiveHeight=1,NegativeHeight=1)
    FalloffSize=0.4
    FlipWidthScaling=false
    OptionalParentItem=(Type=Bone,Name="spine_05")
    RotationOffset=(X=0,Y=-57,Z=-50)
    ExecutePin=@SequenceAggregate.A
   )
   > RigVMVariableNode(Value=@n1.OutputParam) {variable=ads spr}
   n2 = &RigVMFunction_ControlFlowBranch(ExecuteContext=@SequenceAggregate.B, Condition=RigVMVariableNode() {variable=ads}.Value)
   n3 = &RigVMFunction_MathFloatRemap(
    SourceMaximum=-35
    SourceMinimum=-10
    bClamp=true
    Value=&RigVMFunction_MathQuaternionToEuler(RotationOrder=ZYX, Value=&RigUnit_GetTransform(Item=(Type=Bone,Name="head"), Space=LocalSpace).Transform.Rotation).Result.Z
   )
   &RigUnit_ModifyTransforms(
    ItemToModify=((Item=(Type=Bone,Name="spine5_lookat"),Transform=(Rotation=(X=0,Y=0,Z=0,W=1),Translation=(X=0,Y=50,Z=-50),Scale3D=(X=1,Y=1,Z=1))))
    ExecutePin=@n2.True
    Weight=@n3.Result
   )
   > &RigUnit_ModifyTransforms(
    ItemToModify=((Item=(Type=Bone,Name="upperarm_r_lookat"),Transform=(Rotation=(X=0,Y=0,Z=0,W=1),Translation=(X=-10,Y=20,Z=-10),Scale3D=(X=1,Y=1,Z=1))))
    Weight=@n3.Result
   )
   n4 = RigVMVariableNode() {variable=ads spr}
   > &RigUnit_RotationConstraintLocalSpaceOffset(
    AdvancedSettings=(RotationOrderForFilter=XZY)
    Child=(Type=Bone,Name="master_R_front_braid_jnt")
    Filter=(bX=True,bY=True,bZ=True)
    Parents=((Item=(Type=Bone,Name="upperarm_r_lookat"),Weight=1))
    Weight=@n4.Value
   )
   > &RigUnit_AimConstraintLocalSpaceOffset(
    AdvancedSettings=(DebugSettings=(Scale=10,WorldOffset=(Rotation=(X=0,Y=0,Z=0,W=1),Translation=(X=0,Y=0,Z=0),Scale3D=(X=1,Y=1,Z=1))),RotationOrderForFilter=XZY)
    AimAxis=(X=0,Y=-1,Z=0)
    Child=(Type=Bone,Name="master_R_back_braid_jnt")
    Filter=(bX=True,bY=True,bZ=True)
    Parents=((Item=(Type=Bone,Name="upperarm_r_lookat"),Weight=0.5),(Item=(Type=Bone,Name="spine_05"),Weight=0.5))
    UpAxis=(X=1,Y=0,Z=0)
    WorldUp=(Target=(X=0,Y=0,Z=1),Space=(Type=None,Name="None"))
    Weight=@n4.Value
   )
   > &RigUnit_ModifyTransforms(
    ItemToModify=((Item=(Type=Bone,Name="master_R_braid_3_jnt"),Transform=(Rotation=(X=0,Y=0,Z=0,W=1),Translation=(X=0,Y=0,Z=0),Scale3D=(X=1,Y=1,Z=1))))
    ItemToModify.0.Transform.Rotation=&RigVMFunction_MathQuaternionFromEuler(Euler=(X=0,Y=30,Z=-10)).Result
    Weight=@n4.Value
   )
   > &RigUnit_ModifyTransforms(
    ItemToModify=((Item=(Type=Bone,Name="master_L_front_braid_jnt"),Transform=(Rotation=(X=0,Y=0,Z=0,W=1),Translation=(X=0,Y=0,Z=0),Scale3D=(X=1,Y=1,Z=1))))
    ItemToModify.0.Transform.Rotation=&RigVMFunction_MathQuaternionFromEuler(Euler=(X=0,Y=0,Z=15)).Result
    Weight=@n4.Value
   )
   > &RigUnit_ModifyTransforms(
    ItemToModify=((Item=(Type=Bone,Name="master_L_braid_1_jnt"),Transform=(Rotation=(X=0,Y=0,Z=0,W=1),Translation=(X=0,Y=0,Z=0),Scale3D=(X=1,Y=1,Z=1))))
    ItemToModify.0.Transform.Rotation=&RigVMFunction_MathQuaternionFromEuler(Euler=(X=0,Y=0,Z=15)).Result
    Weight=@n4.Value
   )
   &RigUnit_ModifyTransforms(
    ExecutePin=()
    ItemToModify=((Item=(Type=Bone,Name="master_R_braid_1_jnt"),Transform=(Rotation=(X=0,Y=0,Z=0,W=1),Translation=(X=0,Y=0,Z=0),Scale3D=(X=1,Y=1,Z=1))))
    ItemToModify.0.Transform.Rotation=&RigVMFunction_MathQuaternionFromEuler(Euler=(X=0,Y=0,Z=0)).Result
    Weight=@n4.Value
   )
   &RigUnit_RotationConstraintLocalSpaceOffset(
    AdvancedSettings=(RotationOrderForFilter=XZY)
    Child=(Type=Bone,Name="master_R_braid_1_jnt")
    ExecutePin=()
    Filter=(bX=True,bY=True,bZ=True)
    Parents=((Item=(Type=Bone,Name="clavicle_r"),Weight=1),(Item=(Type=Bone,Name="spine_05"),Weight=0))
    Weight=@n4.Value
   )
  ]}
]
hierarchy [
 root {loc=<0 0 0>, rot=<0 0 0 1>}
 pelvis {parent=root, loc=<0 -0.5301 98.0571>, rot=<0.0127 0.707 -0.0127 -0.707>}
 spine_01 {parent=pelvis, loc=<6.1208 -0.9328 0>, rot=<0 0 0.034 -0.9994>}
 spine_02 {parent=spine_01, loc=<6.5996 -0.2095 0>, rot=<0 0 0.0238 -0.9997>}
 spine_03 {parent=spine_02, loc=<6.4914 -0.1237 0>, rot=<0 0 0.0149 -0.9999>}
 spine_04 {parent=spine_03, loc=<6.9598 0.0794 0>, rot=<0 0 -0.0132 -0.9999>}
 spine_05 {parent=spine_04, loc=<6.9987 0 0>, rot=<0 0 -0.0015 -1>}
 neck_01 {parent=spine_05, loc=<11.3138 0.1529 0>, rot=<0 0 -0.2003 -0.9797>}
 neck_02 {parent=neck_01, loc=<4.5627 0 0>, rot=<0 0 0.0504 -0.9987>}
 head {parent=neck_02, loc=<5.3914 0 0>, rot=<0 0 0.115 -0.9934>}
 faceAttach {parent=head, loc=<-0.5242 3.0108 0>, rot=<0.0029 -0.7071 -0.0029 -0.7071>}
 C_brow_mid {parent=faceAttach, loc=<0 8.1598 6.8749>, rot=<0 0 0 -1>}
 C_lip_upper_mid {parent=faceAttach, loc=<0.001 9.1466 -1.8076>, rot=<0.2225 0.0013 -0.0005 -0.9749>}
 C_nose_bridge {parent=faceAttach, loc=<0 9.6532 1.6941>, rot=<-0.2218 0 0 -0.9751>}
 L_brow_mid {parent=faceAttach, loc=<2.9111 7.9235 7.8684>, rot=<-0.0001 -0.0042 0.2553 -0.9669>}
 L_brow_outer {parent=faceAttach, loc=<5.7663 5.8872 8.1811>, rot=<-0.0287 -0.04 0.4491 -0.8921>}
 L_cheek_inner {parent=faceAttach, loc=<3.8372 6.6741 0.1321>, rot=<-0.0706 -0.2697 0.3184 -0.906>}
 L_lip_corner {parent=faceAttach, loc=<2.0682 7.457 -2.406>, rot=<0.003 0.008 -0.0146 -0.9999>}
 L_lip_upper_outer {parent=faceAttach, loc=<1.2095 8.0197 -1.8205>, rot=<0.0359 -0.0728 0.2627 -0.9615>}
 R_brow_mid {parent=faceAttach, loc=<-2.9111 7.9235 7.8684>, rot=<-0.0001 0.0042 -0.2553 -0.9669>}
 R_brow_outer {parent=faceAttach, loc=<-5.7663 5.8872 8.1811>, rot=<-0.0287 0.04 -0.4491 -0.8921>}
 R_cheek_inner {parent=faceAttach, loc=<-3.8372 6.6741 0.1321>, rot=<-0.0706 0.2697 -0.3184 -0.906>}
 R_lip_corner {parent=faceAttach, loc=<-2.0682 7.457 -2.406>, rot=<0.003 -0.008 0.0146 -0.9999>}
 R_lip_upper_outer {parent=faceAttach, loc=<-1.2095 8.0197 -1.8205>, rot=<0.0359 0.0728 -0.2627 -0.9615>}
 master_braids_jnt {parent=head, loc=<14.4316 1.63 0>, rot=<0 0 0 1>}
 master_L_braids_jnt {parent=master_braids_jnt, loc=<-1.2593 -1.5579 -8.006>, rot=<0 0 0 1>}
 master_L_back_braid_jnt {parent=master_L_braids_jnt, loc=<-5.2244 -6.7748 1.8998>, rot=<0.0041 -0.9938 -0.0005 -0.1108>}
 dyn_L_back_braid_1_jnt {parent=master_L_back_braid_jnt, loc=<1.1807 0.1809 -0.2339>, rot=<0 0 0 1>}
 dyn_L_back_braid_2_jnt {parent=dyn_L_back_braid_1_jnt, loc=<6.833 0.3091 -0.8587>, rot=<0.0266 -0.0524 0.013 -0.9982>}
 dyn_L_back_braid_3_jnt {parent=dyn_L_back_braid_2_jnt, loc=<6.2115 -0.8393 -0.4921>, rot=<-0.0185 -0.0298 0.242 -0.9697>}
 master_L_braid_4_jnt {parent=master_L_braids_jnt, loc=<-6.0447 -6.2016 -0.0442>, rot=<-0.0159 -0.9474 -0.3196 -0.0058>}
 dyn_L_braid_4_1_jnt {parent=master_L_braid_4_jnt, loc=<1.7392 0.2702 0>, rot=<0 0 0 1>}
 dyn_L_braid_4_2_jnt {parent=dyn_L_braid_4_1_jnt, loc=<6.9423 0.2078 -0.1394>, rot=<0.09 -0.0146 0.0616 -0.9939>}
 dyn_L_braid_4_3_jnt {parent=dyn_L_braid_4_2_jnt, loc=<7.0612 -1.0118 -0.2631>, rot=<0.0717 -0.0928 0.1719 -0.9781>}
 master_L_braid_3_jnt {parent=master_L_braids_jnt, loc=<-6.8147 -3.6572 -0.8415>, rot=<-0.0314 -0.9993 0.0159 -0.0117>}
 dyn_L_braid_3_1_jnt {parent=master_L_braid_3_jnt, loc=<2.2327 0 0.2514>, rot=<0 0 0 1>}
 dyn_L_braid_3_2_jnt {parent=dyn_L_braid_3_1_jnt, loc=<6.4223 0.3268 -0.1826>, rot=<-0.0443 -0.0542 -0.0583 -0.9958>}
 dyn_L_braid_3_3_jnt {parent=dyn_L_braid_3_2_jnt, loc=<6.8874 0.7554 0.8436>, rot=<0.0523 0.0433 -0.1886 -0.9797>}
 master_L_braid_2_jnt {parent=master_L_braids_jnt, loc=<-8.881 -0.7186 -1.2841>, rot=<-0.0631 -0.9974 0.0024 0.034>}
 dyn_L_braid_2_1_jnt {parent=master_L_braid_2_jnt, loc=<1.8118 0 0>, rot=<0 0 0 1>}
 dyn_L_braid_2_2_jnt {parent=dyn_L_braid_2_1_jnt, loc=<5.2709 0.1148 -0.3606>, rot=<0.1771 0.0003 -0.012 -0.9841>}
 dyn_L_braid_2_3_jnt {parent=dyn_L_braid_2_2_jnt, loc=<7.0446 0.6155 0.8487>, rot=<0.1699 0.0918 -0.0782 -0.9781>}
 dyn_L_braid_2_4_jnt {parent=dyn_L_braid_2_3_jnt, loc=<6.7684 -0.589 0.4977>, rot=<-0.0913 0.0064 0.0389 -0.995>}
 dyn_L_braid_offshoot_jnt {parent=dyn_L_braid_2_2_jnt, loc=<4.1259 0.3595 -1.0363>, rot=<-0.1677 0.0047 -0.0523 -0.9844>}
 master_L_braid_1_jnt {parent=master_L_braids_jnt, loc=<-4.1224 4.4216 -0.6582>, rot=<0.0293 -0.9893 -0.1424 0.0081>}
 dyn_L_braid_1_1_jnt {parent=master_L_braid_1_jnt, loc=<2.2566 -0.518 -0.1985>, rot=<0 0 0 1>}
 dyn_L_braid_1_2_jnt {parent=dyn_L_braid_1_1_jnt, loc=<6.5749 -0.046 -0.0234>, rot=<-0.0031 -0.0494 -0.0568 -0.9972>}
 dyn_L_braid_1_3_jnt {parent=dyn_L_braid_1_2_jnt, loc=<7.2201 0.0402 0.0922>, rot=<-0.0568 0.0819 -0.0766 -0.9921>}
 dyn_L_braid_1_4_jnt {parent=dyn_L_braid_1_3_jnt, loc=<7.1497 0.2741 0.2636>, rot=<0.135 0.061 -0.0385 -0.9882>}
 dyn_L_braid_1_5_jnt {parent=dyn_L_braid_1_4_jnt, loc=<6.4437 0.2762 0.157>, rot=<-0.2566 0.0188 -0.0675 -0.964>}
 master_L_front_braid_jnt {parent=master_L_braids_jnt, loc=<-4.6559 7.2767 0.3545>, rot=<0.0977 -0.994 0.0124 -0.0468>}
 dyn_L_front_braid_1_jnt {parent=master_L_front_braid_jnt, loc=<1.8344 -0.2155 0>, rot=<0 0 0 1>}
 dyn_L_front_braid_2_jnt {parent=dyn_L_front_braid_1_jnt, loc=<3.4372 0.1315 0.0743>, rot=<0.0205 -0.1312 -0.0922 -0.9868>}
 dyn_L_front_braid_3_jnt {parent=dyn_L_front_braid_2_jnt, loc=<3.9357 -0.271 0>, rot=<0 0 0 1>}
 dyn_L_front_braid_4_jnt {parent=dyn_L_front_braid_3_jnt, loc=<5.2117 -0.0878 -0.078>, rot=<0.0029 0.0669 -0.0196 -0.9976>}
 dyn_L_front_braid_5_jnt {parent=dyn_L_front_braid_4_jnt, loc=<3.9291 0 0>, rot=<0 0 0 1>}
 master_R_braids_jnt {parent=master_braids_jnt, loc=<-1.2589 -1.5579 7.8318>, rot=<0 0 1 0>}
 master_R_back_braid_jnt {parent=master_R_braids_jnt, loc=<5.9883 6.6357 -1.1453>, rot=<0.0041 -0.9938 -0.0005 -0.1108>}
 dyn_R_braid_back_1_jnt {parent=master_R_back_braid_jnt, loc=<-1.5045 0 0.415>, rot=<0 0 0 1>}
 dyn_R_braid_back_2_jnt {parent=dyn_R_braid_back_1_jnt, loc=<-6.8991 -0.2951 1.1443>, rot=<0.0266 -0.0524 0.013 -0.9982>}
 dyn_R_braid_back_3_jnt {parent=dyn_R_braid_back_2_jnt, loc=<-6.7873 1.574 0.0188>, rot=<-0.0185 -0.0298 0.242 -0.9697>}
 master_R_braid_4_jnt {parent=master_R_braids_jnt, loc=<7.7541 5.7429 0.0049>, rot=<-0.0159 -0.9474 -0.3196 -0.0058>}
 dyn_R_braid_4_1_jnt {parent=master_R_braid_4_jnt, loc=<0 0 0>, rot=<-0.5099 0 0 -0.8602>}
 dyn_R_braid_4_2_jnt {parent=dyn_R_braid_4_1_jnt, loc=<-6.9815 -0.3664 0.2699>, rot=<0.09 0.0471 0.0424 -0.9939>}
 dyn_R_braid_4_3_jnt {parent=dyn_R_braid_4_2_jnt, loc=<-6.9003 1.0936 -0.7941>, rot=<0.0717 0.1062 0.1639 -0.9781>}
 dyn_R_braid_4_4_jnt {parent=dyn_R_braid_4_3_jnt, loc=<-3.5684 0.7338 -0.2334>, rot=<0 0.0389 0.0213 -0.999>}
 master_R_braid_3_jnt {parent=master_R_braids_jnt, loc=<6.9056 3.6519 1.0281>, rot=<-0.0391 -0.9989 0.0264 -0.0052>}
 dyn_R_braid_3_1_jnt {parent=master_R_braid_3_jnt, loc=<-2.1386 0.0266 -0.2684>, rot=<0 0 0 1>}
 dyn_R_braid_3_2_jnt {parent=dyn_R_braid_3_1_jnt, loc=<-6.4283 -0.2273 0.1079>, rot=<-0.0444 -0.0548 -0.0577 -0.9958>}
 dyn_R_braid_3_3_jnt {parent=dyn_R_braid_3_2_jnt, loc=<-6.8865 -0.6713 -0.9146>, rot=<0.0553 0.0384 -0.1888 -0.9797>}
 master_R_braid_2_jnt {parent=master_R_braids_jnt, loc=<8.4403 0.6465 1.4884>, rot=<-0.0083 -0.9994 0.0006 0.034>}
 dyn_R_braid_2_1_jnt {parent=master_R_braid_2_jnt, loc=<-2.2414 -0.1193 0>, rot=<0.1461 0.008 -0.0543 -0.9877>}
 dyn_R_braid_2_2_jnt {parent=dyn_R_braid_2_1_jnt, loc=<-5.2707 -0.2142 0.3119>, rot=<-0.0436 0.0012 -0.0119 -0.999>}
 dyn_R_braid_2_3_jnt {parent=dyn_R_braid_2_2_jnt, loc=<-7.0447 -0.7351 -0.7474>, rot=<0 0.0624 -0.1033 -0.9927>}
 dyn_R_braid_2_4_jnt {parent=dyn_R_braid_2_3_jnt, loc=<-6.7683 0.2829 -0.7174>, rot=<-0.0913 0.024 0.0313 -0.995>}
 dyn_R_braid_offshoot_jnt {parent=dyn_R_braid_2_2_jnt, loc=<-4.1258 -0.2009 1.0784>, rot=<-0.0936 0.0008 -0.0525 -0.9942>}
 master_R_braid_1_jnt {parent=master_R_braids_jnt, loc=<2.9477 -4.4655 0.6424>, rot=<0.0293 -0.9893 -0.1424 0.0081>}
 dyn_R_1_1_braid_jnt {parent=master_R_braid_1_jnt, loc=<-3.4312 0.5483 0>, rot=<0 0 0 1>}
 dyn_R_1_2_braid_jnt {parent=dyn_R_1_1_braid_jnt, loc=<-6.8732 -0.205 0.3308>, rot=<-0.0031 -0.0494 -0.0568 -0.9972>}
 dyn_R_1_3_braid_jnt {parent=dyn_R_1_2_braid_jnt, loc=<-6.9235 0.1249 -0.0018>, rot=<-0.0568 0.0819 -0.0766 -0.9921>}
 dyn_R_1_4_braid_jnt {parent=dyn_R_1_3_braid_jnt, loc=<-6.9916 0.0558 -0.0205>, rot=<-0.024 0.0541 -0.0477 -0.9971>}
 dyn_R_1_5_braid_jnt {parent=dyn_R_1_4_braid_jnt, loc=<-6.6396 -0.9385 0.3961>, rot=<-0.0999 0.0079 -0.0696 -0.9925>}
 master_R_front_braid_jnt {parent=master_R_braids_jnt, loc=<2.1018 -8.1847 -0.8498>, rot=<0.0977 -0.994 0.0124 -0.0468>}
 dyn_R_front_braid_1_jnt {parent=master_R_front_braid_jnt, loc=<-1.7875 0.0112 -0.4321>, rot=<0 0 0 1>}
 dyn_R_front_braid_2_jnt {parent=dyn_R_front_braid_1_jnt, loc=<-3.438 0.1975 -0.0742>, rot=<0.0205 -0.1312 -0.0922 -0.9868>}
 dyn_R_front_braid_3_jnt {parent=dyn_R_front_braid_2_jnt, loc=<-5.1397 0.271 -0.4524>, rot=<0 0 0 -1>}
 dyn_R_front_braid_4_jnt {parent=dyn_R_front_braid_3_jnt, loc=<-5.0503 0.0909 -0.29>, rot=<0.0029 0.0669 -0.0196 -0.9976>}
 dyn_R_front_braid_5_jnt {parent=dyn_R_front_braid_4_jnt, loc=<-5.2273 0.3706 0>, rot=<0 0 0 -1>}
 clavicle_l {parent=spine_05, loc=<5.4831 -0.1053 -3.2946>, rot=<0.0465 -0.7289 -0.009 -0.6829>}
 upperarm_l {parent=clavicle_l, loc=<9.8658 -1.2136 -0.4105>, rot=<0.0164 -0.3245 -0.0269 -0.9454>}
 clavicle_r {parent=spine_05, loc=<5.4831 -0.1053 3.2946>, rot=<0.7289 0.0465 -0.6829 0.009>}
 upperarm_r {parent=clavicle_r, loc=<-9.8658 1.2136 0.4105>, rot=<0.0164 -0.3245 -0.0269 -0.9454>}
 upperarm_r_lookat {parent=upperarm_r, loc=<0 0 0>, rot=<0 0 0 1>}
 spine5_lookat {parent=spine_05, loc=<0 0 0>, rot=<0 0 0 1>}
]
variables [
 ads {type=bool, default=False}
 ads spr {type=double, default=0}
]
