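# Physics asset for the face-accessory part of the female medium Dazzle player body, setting up physical motion for its braids.
# Capsule bodies cover left and right front, side and back braid chains, two offshoots, and the head, neck, upper spine, clavicles and right upper arm.
# Constraints chain each braid from the head, with offshoots branching mid-braid.
# All joints have zero twist limit, and back braids get a wider swing than the others.
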PhysicsAsset /Game/Characters/Player/Female/Medium/Bodies/F_MED_Dazzle/Meshes/Parts/F_MED_Dazzle_FaceAcc_Physics
bodies {
 clavicle_l [{
   shapeType=Capsule
   length=11.25
   radius=5.255
   shapeName=clavicle_l_capsule
   center=<10.4689 -2.2317 0.8692>
   rotation=<rot 6.7173 -71.1183 -108.8818>
  }]
 clavicle_r [{
   shapeType=Capsule
   length=11.25
   radius=5.255
   shapeName=clavicle_r_capsule
   center=<-10.0378 0 -3.6535>
   rotation=<rot 3.4046 -80.5919 -109.7198>
  }]
 dyn_L_back_braid_1_jnt [{
   shapeType=Capsule
   length=3.7984
   radius=2.656
   shapeName=dyn_L_back_braid_1_jnt_capsule
   center=<5.4612 0 -0.1119>
   rotation=<rot 80 0 -0.0006>
  }]
 dyn_L_back_braid_2_jnt [{
   shapeType=Capsule
   length=5.2216
   radius=2.5337
   shapeName=dyn_L_back_braid_2_jnt_capsule
   center=<6.2306 0 -0.1119>
   rotation=<rot 80 0.0003 9.9997>
  }]
 dyn_L_back_braid_3_jnt [{
   shapeType=Capsule
   length=4.1095
   radius=2.5228
   shapeName=dyn_L_back_braid_3_jnt_capsule
   center=<3.7959 0 -0.1119>
   rotation=<rot 80 0.0004 9.9999>
  }]
 dyn_L_braid_1_1_jnt [{shapeType=Capsule, length=4, radius=1.25, shapeName=dyn_L_braid_1_1_jnt_capsule, center=<2.6786 0 0>, rotation=<rot -90 -179.999 179.9992>}]
 dyn_L_braid_1_2_jnt [{shapeType=Capsule, length=4, radius=1.25, shapeName=dyn_L_braid_1_2_jnt_capsule, center=<2.6786 0 0>, rotation=<rot -90 -0.0002 0.0003>}]
 dyn_L_braid_1_3_jnt [{shapeType=Capsule, length=4, radius=1.25, shapeName=dyn_L_braid_1_3_jnt_capsule, center=<2.6786 0 0>, rotation=<rot -90 -174.9488 174.9489>}]
 dyn_L_braid_1_4_jnt [{shapeType=Capsule, length=4, radius=1.25, shapeName=dyn_L_braid_1_4_jnt_capsule, center=<2.6786 0 0>, rotation=<rot -90 160.5288 -160.5287>}]
 dyn_L_braid_1_5_jnt [{shapeType=Capsule, length=4, radius=1.25, shapeName=dyn_L_braid_1_5_jnt_capsule, center=<2.6786 0 0>, rotation=<rot -90 180 -179.9999>}]
 dyn_L_braid_2_1_jnt [{
   shapeType=Capsule
   length=3.75
   radius=1.25
   shapeName=dyn_L_braid_2_1_jnt_capsule
   center=<2.5549 0.1387 0.1299>
   rotation=<rot 90 180 179.9998>
  }]
 dyn_L_braid_2_2_jnt [{
   shapeType=Capsule
   length=3.75
   radius=1.25
   shapeName=dyn_L_braid_2_2_jnt_capsule
   center=<2.6592 0 0.3385>
   rotation=<rot 79.9999 -179.9999 179.9999>
  }]
 dyn_L_braid_2_3_jnt [{
   shapeType=Capsule
   length=3.75
   radius=1.75
   shapeName=dyn_L_braid_2_3_jnt_capsule
   center=<2.6873 0 0.1791>
   rotation=<rot 80 -179.9998 -180>
  }]
 dyn_L_braid_2_4_jnt [{
   shapeType=Capsule
   length=3.75
   radius=2.25
   shapeName=dyn_L_braid_2_4_jnt_capsule
   center=<3.093 0 1.0248>
   rotation=<rot 79.9999 -179.9998 -179.9999>
  }]
 dyn_L_braid_3_1_jnt [{
   shapeType=Capsule
   length=2.75
   radius=1.5
   shapeName=dyn_L_braid_3_1_jnt_capsule
   center=<3.1938 0 -0.1445>
   rotation=<rot -90 144.7355 -144.7353>
  }]
 dyn_L_braid_3_2_jnt [{
   shapeType=Capsule
   length=2.75
   radius=1.5
   shapeName=dyn_L_braid_3_2_jnt_capsule
   center=<2.2563 0 -0.3928>
   rotation=<rot -90 -166.7373 166.7375>
  }]
 dyn_L_braid_3_3_jnt [{
   shapeType=Capsule
   length=4.25
   radius=1.5
   shapeName=dyn_L_braid_3_3_jnt_capsule
   center=<2.6267 0 -0.029>
   rotation=<rot -80.0001 0 0.0001>
  }]
 dyn_L_braid_offshoot_jnt [{
   shapeType=Capsule
   length=3.75
   radius=1.25
   shapeName=dyn_L_braid_offshoot_jnt_capsule
   center=<3.3411 0 -0.1558>
   rotation=<rot 90 180 179.9998>
  }]
 dyn_L_front_braid_1_jnt [{shapeType=Capsule, length=2.75, radius=1.25, shapeName=dyn_L_front_braid_1_jnt_capsule, center=<2.4758 0 0.1178>, rotation=<rot 80 0.0001 0>}]
 dyn_L_front_braid_2_jnt [{
   shapeType=Capsule
   length=2.75
   radius=1.25
   shapeName=dyn_L_front_braid_2_jnt_capsule
   center=<2.381 0 -0.0256>
   rotation=<rot 90 0 -0.0002>
  }]
 dyn_L_front_braid_3_jnt [{
   shapeType=Capsule
   length=2.75
   radius=1.25
   shapeName=dyn_L_front_braid_3_jnt_capsule
   center=<3.1543 0 -0.1551>
   rotation=<rot 90 90 89.9999>
  }]
 dyn_L_front_braid_4_jnt [{shapeType=Capsule, length=3, radius=1, shapeName=dyn_L_front_braid_4_jnt_capsule, center=<2.381 0 -0.4198>, rotation=<rot 80 0.0001 0>}]
 dyn_L_front_braid_5_jnt [{shapeType=Capsule, length=3, radius=1, shapeName=dyn_L_front_braid_5_jnt_capsule, center=<2.381 0 -0.4198>, rotation=<rot 80 0.0001 0>}]
 dyn_R_1_1_braid_jnt [{
   shapeType=Capsule
   length=3.25
   radius=1.75
   shapeName=dyn_R_1_1_braid_jnt_capsule
   center=<-2.8389 0 -0.5006>
   rotation=<rot -90 -1.5638 1.564>
  }]
 dyn_R_1_2_braid_jnt [{
   shapeType=Capsule
   length=4.25
   radius=1.75
   shapeName=dyn_R_1_2_braid_jnt_capsule
   center=<-2.7383 0 -1.8773>
   rotation=<rot -90 -179.9999 -179.9999>
  }]
 dyn_R_1_3_braid_jnt [{shapeType=Capsule, length=3, radius=1.5, shapeName=dyn_R_1_3_braid_jnt_capsule, center=<-2.8389 0 -0.5006>, rotation=<rot -80 0.0001 0>}]
 dyn_R_1_4_braid_jnt [{
   shapeType=Capsule
   length=3.75
   radius=1.5
   shapeName=dyn_R_1_4_braid_jnt_capsule
   center=<-2.8389 -0.1085 0.32>
   rotation=<rot -20 -89.9994 90>
  }]
 dyn_R_1_5_braid_jnt [{
   shapeType=Capsule
   length=3
   radius=1.5
   shapeName=dyn_R_1_5_braid_jnt_capsule
   center=<-2.3502 0 0.0219>
   rotation=<rot -90 179.9926 -179.9924>
  }]
 dyn_R_braid_2_1_jnt [{shapeType=Capsule, length=4, radius=1, shapeName=dyn_R_braid_2_1_jnt_capsule, center=<-3.204 0 0>, rotation=<rot 90 0 -0.0001>}]
 dyn_R_braid_2_2_jnt [{shapeType=Capsule, length=4, radius=1.5, shapeName=dyn_R_braid_2_2_jnt_capsule, center=<-4.5958 0 0>, rotation=<rot 90 179.9999 179.9998>}]
 dyn_R_braid_2_3_jnt [{
   shapeType=Capsule
   length=3.75
   radius=1.25
   shapeName=dyn_R_braid_2_3_jnt_capsule
   center=<-5.1077 0 -0.3357>
   rotation=<rot 80 -179.9999 -180>
  }]
 dyn_R_braid_2_4_jnt [{
   shapeType=Capsule
   length=3.75
   radius=1.25
   shapeName=dyn_R_braid_2_4_jnt_capsule
   center=<-3.801 0 -0.6846>
   rotation=<rot 80 -179.9998 -180>
  }]
 dyn_R_braid_3_1_jnt [{shapeType=Capsule, length=4, radius=1, shapeName=dyn_R_braid_3_1_jnt_capsule, center=<-2.7958 0 0>, rotation=<rot 90 0.0099 0.0099>}]
 dyn_R_braid_3_2_jnt [{
   shapeType=Capsule
   length=4
   radius=1
   shapeName=dyn_R_braid_3_2_jnt_capsule
   center=<-2.7958 0 -0.3624>
   rotation=<rot 90 179.9604 179.9604>
  }]
 dyn_R_braid_3_3_jnt [{shapeType=Capsule, length=4, radius=1, shapeName=dyn_R_braid_3_3_jnt_capsule, center=<-2.7958 0 -0.9249>, rotation=<rot 90 90 90>}]
 dyn_R_braid_back_1_jnt [{
   shapeType=Capsule
   length=3.1128
   radius=2.2531
   shapeName=dyn_R_braid_back_1_jnt_capsule
   center=<-4.642 0 0.0984>
   rotation=<rot 80 0.0001 -0.0001>
  }]
 dyn_R_braid_back_2_jnt [{
   shapeType=Capsule
   length=3.6583
   radius=2.5841
   shapeName=dyn_R_braid_back_2_jnt_capsule
   center=<-3.5011 0 0.0984>
   rotation=<rot 80 0.0002 9.9999>
  }]
 dyn_R_braid_back_3_jnt [{
   shapeType=Capsule
   length=4.2794
   radius=2.7453
   shapeName=dyn_R_braid_back_3_jnt_capsule
   center=<-4.073 0 0.0984>
   rotation=<rot 80 -179.9995 -169.9999>
  }]
 dyn_R_braid_offshoot_jnt [{shapeType=Capsule, length=4, radius=1, shapeName=dyn_R_braid_offshoot_jnt_capsule, center=<-3.204 0 0>, rotation=<rot 90 6.7213 6.7212>}]
 dyn_R_front_braid_1_jnt [{shapeType=Capsule, length=3, radius=1, shapeName=dyn_R_front_braid_1_jnt_capsule, center=<-2.881 0 0>, rotation=<rot 90 0 -0.0002>}]
 dyn_R_front_braid_2_jnt [{shapeType=Capsule, length=3, radius=1, shapeName=dyn_R_front_braid_2_jnt_capsule, center=<-2.881 0 0>, rotation=<rot 90 180 179.9998>}]
 dyn_R_front_braid_3_jnt [{shapeType=Capsule, length=3, radius=1, shapeName=dyn_R_front_braid_3_jnt_capsule, center=<-2.881 0 0>, rotation=<rot 90 0.0001 -0.0001>}]
 dyn_R_front_braid_4_jnt [{shapeType=Capsule, length=3, radius=1, shapeName=dyn_R_front_braid_4_jnt_capsule, center=<-2.881 0 0>, rotation=<rot 90 0.0001 -0.0001>}]
 dyn_R_front_braid_5_jnt [{shapeType=Capsule, length=3, radius=1, shapeName=dyn_R_front_braid_5_jnt_capsule, center=<-2.881 0 0>, rotation=<rot 90 179.9605 179.9603>}]
 head [{shapeType=Capsule, length=6.25, radius=8.5, shapeName=head_capsule, center=<4.1543 1.6025 0>, rotation=<rot 90 0 0>}]
 neck_01 [
  {shapeType=Capsule, length=60.5, radius=2.5, shapeName=neck_01_capsule, center=<0 -0.6879 0>}
  {shapeType=Capsule, length=5.25, radius=4.75, shapeName=neck_01_capsule, center=<2.408 0 0>, rotation=<rot -90 180 179.9999>}
 ]
 neck_02 [{shapeType=Capsule, length=5.25, radius=4.75, shapeName=neck_02_capsule, center=<2.408 0 0>, rotation=<rot -90 -179.9996 179.9995>}]
 spine_03 [{shapeType=Capsule, length=12.75, radius=7, shapeName=spine_04_capsule, center=<-0.8829 1.2598 0>}]
 spine_04 [{shapeType=Capsule, length=12, radius=9.5, shapeName=spine_04_capsule, center=<0 1.6532 0>}]
 spine_05 [{shapeType=Capsule, length=34, radius=7.5, shapeName=spine_05_capsule, center=<2.746 -0.8671 0>}]
 upperarm_r [{shapeType=Capsule, length=27.25, radius=7.255, shapeName=clavicle_r_capsule, center=<-1.0156 -11.3724 -3.817>, rotation=<rot 40 0 69.9991>}]
}
constraints [
 {bone1Name=dyn_R_front_braid_5_jnt, bone2Name=dyn_R_front_braid_4_jnt, swing1LimitDegrees=25, swing2LimitDegrees=15, twistLimitDegrees=0}
 {bone1Name=dyn_L_back_braid_1_jnt, bone2Name=head, swing1LimitDegrees=45, swing2LimitDegrees=15, twistLimitDegrees=0}
 {bone1Name=dyn_L_back_braid_2_jnt, bone2Name=dyn_L_back_braid_1_jnt, swing1LimitDegrees=45, swing2LimitDegrees=15, twistLimitDegrees=0}
 {bone1Name=dyn_L_back_braid_3_jnt, bone2Name=dyn_L_back_braid_2_jnt, swing1LimitDegrees=45, swing2LimitDegrees=15, twistLimitDegrees=0}
 {bone1Name=dyn_L_braid_1_1_jnt, bone2Name=head, swing1LimitDegrees=25, swing2LimitDegrees=15, twistLimitDegrees=0}
 {bone1Name=dyn_L_braid_1_2_jnt, bone2Name=dyn_L_braid_1_1_jnt, swing1LimitDegrees=25, swing2LimitDegrees=15, twistLimitDegrees=0}
 {bone1Name=dyn_L_braid_1_3_jnt, bone2Name=dyn_L_braid_1_2_jnt, swing1LimitDegrees=25, swing2LimitDegrees=15, twistLimitDegrees=0}
 {bone1Name=dyn_L_braid_1_4_jnt, bone2Name=dyn_L_braid_1_3_jnt, swing1LimitDegrees=25, swing2LimitDegrees=15, twistLimitDegrees=0}
 {bone1Name=dyn_L_braid_1_5_jnt, bone2Name=dyn_L_braid_1_4_jnt, swing1LimitDegrees=25, swing2LimitDegrees=15, twistLimitDegrees=0}
 {bone1Name=dyn_L_braid_2_1_jnt, bone2Name=head, swing1LimitDegrees=25, swing2LimitDegrees=15, twistLimitDegrees=0}
 {bone1Name=dyn_L_braid_2_2_jnt, bone2Name=dyn_L_braid_2_1_jnt, swing1LimitDegrees=25, swing2LimitDegrees=15, twistLimitDegrees=0}
 {bone1Name=dyn_L_braid_2_3_jnt, bone2Name=dyn_L_braid_2_2_jnt, swing1LimitDegrees=25, swing2LimitDegrees=15, twistLimitDegrees=0}
 {bone1Name=dyn_L_braid_2_4_jnt, bone2Name=dyn_L_braid_2_3_jnt, swing1LimitDegrees=25, swing2LimitDegrees=15, twistLimitDegrees=0}
 {bone1Name=dyn_L_braid_offshoot_jnt, bone2Name=dyn_L_braid_2_2_jnt, swing1LimitDegrees=25, swing2LimitDegrees=15, twistLimitDegrees=0}
 {bone1Name=dyn_L_braid_3_1_jnt, bone2Name=head, swing1LimitDegrees=25, swing2LimitDegrees=15, twistLimitDegrees=0}
 {bone1Name=dyn_L_braid_3_2_jnt, bone2Name=dyn_L_braid_3_1_jnt, swing1LimitDegrees=25, swing2LimitDegrees=15, twistLimitDegrees=0}
 {bone1Name=dyn_L_braid_3_3_jnt, bone2Name=dyn_L_braid_3_2_jnt, swing1LimitDegrees=25, swing2LimitDegrees=15, twistLimitDegrees=0}
 {bone1Name=dyn_L_front_braid_1_jnt, bone2Name=head, swing1LimitDegrees=25, swing2LimitDegrees=15, twistLimitDegrees=0}
 {bone1Name=dyn_L_front_braid_2_jnt, bone2Name=dyn_L_front_braid_1_jnt, swing1LimitDegrees=25, swing2LimitDegrees=15, twistLimitDegrees=0}
 {bone1Name=dyn_L_front_braid_3_jnt, bone2Name=dyn_L_front_braid_2_jnt, swing1LimitDegrees=25, swing2LimitDegrees=15, twistLimitDegrees=0}
 {bone1Name=dyn_L_front_braid_4_jnt, bone2Name=dyn_L_front_braid_3_jnt, swing1LimitDegrees=25, swing2LimitDegrees=15, twistLimitDegrees=0}
 {bone1Name=dyn_L_front_braid_5_jnt, bone2Name=dyn_L_front_braid_4_jnt, swing1LimitDegrees=25, swing2LimitDegrees=15, twistLimitDegrees=0}
 {bone1Name=dyn_R_braid_back_3_jnt, bone2Name=dyn_R_braid_back_2_jnt, swing1LimitDegrees=45, swing2LimitDegrees=15, twistLimitDegrees=0}
 {bone1Name=dyn_R_1_1_braid_jnt, bone2Name=head, swing1LimitDegrees=25, swing2LimitDegrees=15, twistLimitDegrees=0}
 {bone1Name=dyn_R_1_2_braid_jnt, bone2Name=dyn_R_1_1_braid_jnt, swing1LimitDegrees=25, swing2LimitDegrees=15, twistLimitDegrees=0}
 {bone1Name=dyn_R_1_3_braid_jnt, bone2Name=dyn_R_1_2_braid_jnt, swing1LimitDegrees=25, swing2LimitDegrees=15, twistLimitDegrees=0}
 {bone1Name=dyn_R_1_4_braid_jnt, bone2Name=dyn_R_1_3_braid_jnt, swing1LimitDegrees=25, swing2LimitDegrees=15, twistLimitDegrees=0}
 {bone1Name=dyn_R_1_5_braid_jnt, bone2Name=dyn_R_1_4_braid_jnt, swing1LimitDegrees=25, swing2LimitDegrees=15, twistLimitDegrees=0}
 {bone1Name=dyn_R_braid_2_1_jnt, bone2Name=head, swing1LimitDegrees=25, swing2LimitDegrees=15, twistLimitDegrees=0}
 {bone1Name=dyn_R_braid_2_2_jnt, bone2Name=dyn_R_braid_2_1_jnt, swing1LimitDegrees=25, swing2LimitDegrees=15, twistLimitDegrees=0}
 {bone1Name=dyn_R_braid_2_3_jnt, bone2Name=dyn_R_braid_2_2_jnt, swing1LimitDegrees=25, swing2LimitDegrees=15, twistLimitDegrees=0}
 {bone1Name=dyn_R_braid_2_4_jnt, bone2Name=dyn_R_braid_2_3_jnt, swing1LimitDegrees=25, swing2LimitDegrees=15, twistLimitDegrees=0}
 {bone1Name=dyn_R_braid_offshoot_jnt, bone2Name=dyn_R_braid_2_2_jnt, swing1LimitDegrees=25, swing2LimitDegrees=15, twistLimitDegrees=0}
 {bone1Name=dyn_R_braid_3_1_jnt, bone2Name=head, swing1LimitDegrees=25, swing2LimitDegrees=15, twistLimitDegrees=0}
 {bone1Name=dyn_R_braid_3_2_jnt, bone2Name=dyn_R_braid_3_1_jnt, swing1LimitDegrees=25, swing2LimitDegrees=15, twistLimitDegrees=0}
 {bone1Name=dyn_R_braid_3_3_jnt, bone2Name=dyn_R_braid_3_2_jnt, swing1LimitDegrees=25, swing2LimitDegrees=15, twistLimitDegrees=0}
 {bone1Name=dyn_R_front_braid_1_jnt, bone2Name=head, swing1LimitDegrees=25, swing2LimitDegrees=15, twistLimitDegrees=0}
 {bone1Name=dyn_R_front_braid_2_jnt, bone2Name=dyn_R_front_braid_1_jnt, swing1LimitDegrees=25, swing2LimitDegrees=15, twistLimitDegrees=0}
 {bone1Name=dyn_R_front_braid_3_jnt, bone2Name=dyn_R_front_braid_2_jnt, swing1LimitDegrees=25, swing2LimitDegrees=15, twistLimitDegrees=0}
 {bone1Name=dyn_R_front_braid_4_jnt, bone2Name=dyn_R_front_braid_3_jnt, swing1LimitDegrees=25, swing2LimitDegrees=15, twistLimitDegrees=0}
 {bone1Name=dyn_R_braid_back_2_jnt, bone2Name=dyn_R_braid_back_1_jnt, swing1LimitDegrees=45, swing2LimitDegrees=15, twistLimitDegrees=0}
 {bone1Name=dyn_R_braid_back_1_jnt, bone2Name=head, swing1LimitDegrees=45, swing2LimitDegrees=15, twistLimitDegrees=0}
]
