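# Animation Blueprint for the F_MED_Dazzle face accessory mesh, posing and simulating the character's braids and glasses.
# Copies the attached parent's pose into a Control Rig fed the targeting flag, then runs RigidBody physics relative to spine_05, at half alpha during skydive floating.
# Additive bone offsets then adjust back braids while floating, glasses while targeting, and two braid joints in the front end.
# Dynamics reset when the parachute opens or skydiving ends.
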AnimBlueprint /BRCosmetics/Characters/Player/Female/Medium/Bodies/F_MED_Dazzle/Meshes/Parts/F_MED_Dazzle_FaceAcc_AnimBP
target_skeleton=/Game/Characters/Player/Female/Medium/Bodies/F_MED_Dazzle/Meshes/Parts/F_MED_Dazzle_FaceAcc_Skeleton
graphs [
 AnimGraph {nodes [
   n1 = Control Rig!(Source=Copy Pose From Mesh() {bUseAttachedParent=true}, Alpha, ads:bool=Get ads() -> (ads:bool)) -> (Pose) {
    alphaBoolBlend {blendOption=Linear}
    controlRigAssetReference {blueprintRigClass=/Game/Characters/Player/Female/Medium/Bodies/F_MED_Dazzle/Meshes/Parts/F_MED_Dazzle_FaceAcc_CtrlRig.F_MED_Dazzle_FaceAcc_CtrlRig_C}
   }
   n2 = RigidBody(ComponentPose=Local To Component(LocalPose=@n1), Alpha=SelectFloat(A=0.5, B=1, bPickA=Get bIsSkydiveFloating())) {
    bClampLinearTranslationLimitToRefPose=true
    bOverrideWorldGravity=true
    overridePhysicsAsset=/Game/Characters/Player/Female/Medium/Bodies/F_MED_Dazzle/Meshes/Parts/F_MED_Dazzle_FaceAcc_Physics
    simulationSpace=BaseBoneSpace
    alphaBoolBlend {blendOption=Linear}
    baseBoneRef {boneName=spine_05}
    componentAppliedLinearAccClamp=<3000 3000 3000>
    componentLinearAccScale=<0.3 0.2 0.2>
    componentLinearVelScale=<0.3 0.2 0.2>
    overrideWorldGravity {z=-980}
    simSpaceSettings {maxAngularAcceleration=1000, maxAngularVelocity=3, maxLinearAcceleration=250, maxLinearVelocity=500, velocityScaleZ=0, worldAlpha=1}
   }
   n3 = Get Skydiving() -> (Skydiving:double)
   n4 = "Transform (Modify) Bone - Bone: master_R_back_braid_jnt"(ComponentPose=@n2, Alpha=@n3, Translation="0,0,0", Rotation="0,-50,-10", Scale="1,1,1") {
    rotationMode=BMM_Additive
    rotationSpace=BCS_BoneSpace
    alphaBoolBlend {blendOption=Linear}
    boneToModify {boneName=master_R_back_braid_jnt}
    rotation {roll=-10, yaw=-50}
   }
   n5 = "Transform (Modify) Bone - Bone: master_L_back_braid_jnt"(ComponentPose=@n4, Alpha=@n3, Translation="0,0,0", Rotation="0,-65,10", Scale="1,1,1") {
    rotationMode=BMM_Additive
    rotationSpace=BCS_BoneSpace
    alphaBoolBlend {blendOption=Linear}
    boneToModify {boneName=master_L_back_braid_jnt}
    rotation {roll=10, yaw=-65}
   }
   n6 = "Transform (Modify) Bone - Bone: anim_glasses_jnt"(ComponentPose=@n5, Alpha="To Float (Boolean)"(InBool=Get ads()), Translation="0,0.5,0", Rotation="0,0,0", Scale="1,1,1") {
    translationMode=BMM_Additive
    translationSpace=BCS_BoneSpace
    alphaBoolBlend {blendOption=Linear}
    alphaScaleBiasClamp {bInterpResult=true, interpSpeedDecreasing=5}
    boneToModify {boneName=anim_glasses_jnt}
    translation {y=0.5}
   }
   n7 = Get Front End() -> (Front End:double)
   n8 = "Transform (Modify) Bone - Bone: master_L_braid_3_jnt"(ComponentPose=@n6, Alpha=@n7, Translation="0,0,0", Rotation="0,-5,0", Scale="1,1,1") {
    rotationMode=BMM_Additive
    rotationSpace=BCS_ParentBoneSpace
    alphaBoolBlend {blendOption=Linear}
    boneToModify {boneName=master_L_braid_3_jnt}
    rotation {yaw=-5}
   }
   n9 = "Transform (Modify) Bone - Bone: master_R_braid_1_jnt"(ComponentPose=@n8, Alpha=@n7, Translation="0,0,-0.5", Rotation="4,-1.5,0", Scale="1,1,1") {
    rotationMode=BMM_Additive
    rotationSpace=BCS_ParentBoneSpace
    translationMode=BMM_Additive
    translationSpace=BCS_BoneSpace
    alphaBoolBlend {blendOption=Linear}
    boneToModify {boneName=master_R_braid_1_jnt}
    rotation {pitch=4, yaw=-1.5}
    translation {z=-0.5}
   }
   Output Pose(Result=Component To Local(ComponentPose=@n9)) {layerGroup=None}
  ]}
 EventGraph {nodes [
   Event BlueprintUpdateAnimation()
   > Set Skydiving(Skydiving:double="To Float (Boolean)"(InBool=Get bIsSkydiveFloating())) -> (Output_Get:double)
   > Set ads(ads:bool=Get bIsTargeting()) -> (Output_Get:bool)
   > Set Front End(Front End:double="To Float (Boolean)"(InBool=Get bIsInFrontEnd())) -> (Output_Get:double)
   n1 = Event OnEndSkydiving()
   ResetDynamics(execute=[@n1.then, @n2.then], InTeleportType=ResetPhysics)
   n2 = Event OnParachuteOpened()
  ]}
]
variables [
 ads {type=boolean}
 Front End {type=number}
 Skydiving {type=number}
]
